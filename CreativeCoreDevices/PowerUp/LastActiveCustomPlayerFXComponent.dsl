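; Player-state component that shows power-up device custom VFX and audio on the player's character, with team-based color and visibility.
; The server replicates effect actors in five RepNotify slots, kept separate for bandwidth, with a toggled bool.
; Clients read details through the device effects interface, queue async loads, then attach reusable third-person, first-person and audio components.
; Pawn and team changes refresh effects, death clears them, and minigame play-mode changes destroy the component.
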(event Custom|Bind%20to%20Minigame
  ;[0 0.261 1 0.1] Minigame Binding: Both the server and clients want to bind to the minigame. The Server binds to the play mode changing, and destroys this component (and any spawned effects) when that happens. The client binds to team changes. Minigame isn't guaranteed to be ready when this component replicates, so if it's not, wait and retry until it is.
  (Variables|CachedData|SetMinigame (Minigame|GetMinigame (Components|GetOwner) :!class /Script/FortniteGame.MinigameBlueprintLibrary))
  (Networking|SwitchHasAuthority
    (:Authority
      (bind _minigame (Variables|CachedData|GetMinigame)
        (:then
          (Minigame|BindEventtoOnPlayModeChangedDelegate _minigame (EventDispatchers|CreateEvent :!event Custom|OnMinigameChanged)))
        (:else
          _)))
    (:Remote
      (Networking|CanExecuteCosmeticEvents
        (:True
          (bind _minigame_2 (Variables|CachedData|GetMinigame)
            (:then
              (Minigame|BindEventtoMinigameTeamConfigUpdated _minigame_2 (EventDispatchers|CreateEvent :!event Custom|OnTeamConfigChanged)))
            (:else
              _))))))
  (Utilities|FlowControl|Delay 1.0)
  (CallFunction|BindToMinigame))

(event EventBeginPlay
  (bind _owning_player_state (Variables|CachedData|GetOwning%20Player%20State))
  ;[0 1 0.1954 0.44] Creation.
  (CallFunction|BindToMinigame)
  (Networking|CanExecuteCosmeticEvents
    (:True
      (bind _asfort_player_state_zone (Utilities|Casting|CastToFortPlayerStateZone (Components|GetOwner)))
      (bind _output_get (Variables|CachedData|SetOwning%20Player%20State _asfort_player_state_zone))
      (bind _ascharacter (Utilities|Casting|CastToCharacter (PlayerState|GetPawn _output_get))
        (:then
          (Variables|CachedData|SetTargetCharacter _ascharacter))
        (:CastFailed
          _))
      ;[0 1 0.1753 0.245] Bind to Changes in the Owning PlayerState
      (EventDispatchers|AssignOnPawnDiedDelegate _owning_player_state (EventDispatchers|CreateEvent :!event Custom|OnPawnDied))
      (Pawn|BindEventtoOnCurrentPawnChanged _owning_player_state (EventDispatchers|CreateEvent :!event Reinitialize%20All%20Effects))
      ;[0.02505 0.4375 0 0.164] Cache the teams of both the player that this is applied to and the local player, as we'll be doing team checks against them later for visibility/color.
      (Variables|CachedData|SetOwnerTeam (Team|GetActorTeam (Variables|CachedData|GetOwning%20Player%20State)))
      (Variables|CachedData|SetLocalTeam (Team|GetActorTeam (Fortnite|GetFirstLocalFortPlayerController true))))))

(event Custom|OnTeamConfigChanged
  (bind _owning_player_state (Variables|CachedData|GetOwning%20Player%20State))
  (bind _returnvalue (Team|GetActorTeam _owning_player_state))
  (bind _returnvalue_2 (Fortnite|GetFirstLocalFortPlayerController true))
  (bind _returnvalue_3 (Team|GetActorTeam _returnvalue_2))
  ;[0.03552 0.5552 1 0.348] When either the local player (viewer) or the owning player (VFX actor) change teams, update the color and visibility, since they might have changed.
  (if (!= (Variables|CachedData|GetOwnerTeam) _returnvalue)
    (Variables|CachedData|SetOwnerTeam _returnvalue)
    (elif (!= (Variables|CachedData|GetLocalTeam) _returnvalue_3)
      (Variables|CachedData|SetLocalTeam _returnvalue_3)))
  (Default|UpdateColorandVisibilityForAll))

(event Custom|Load%20Queued%20Effects
  (bind _loaded_details (Variables|LoadedDetails|GetLoaded%20Details))
  (bind _audio_6_c55694dd4bad4c0b07d0b6ba6026f55b (Utilities|Struct|BreakCustomFXDetails _loaded_details))
  (bind _queuedfx (Variables|Default|GetQueuedFX))
  ;[0 0.6864 0.7396 0.292] Async Loading Queued Effects - Load the effects form the first element in our Queued FX Array, cache a reference then spawn it from the cached details. Once done, clear the Cached details, pop the first element off (we're done with it) then restart.
  (Networking|CanExecuteCosmeticEvents
    (:True
      (if (and (Utilities|Array|IsValidIndex _queuedfx) (not (Utilities|DetailsHaveValidRef (Variables|LoadedDetails|GetLoaded%20Details))))
        (bind _output_get (Variables|LoadedDetails|SetLoaded%20Details (Utilities|Array|Get(acopy) _queuedfx)))
        (if (Utilities|IsValidSoftObjectReference (Utilities|Struct|BreakCustomFXDetails _output_get))
          (bind _object (Utilities|AsyncLoadAsset (Utilities|Struct|BreakCustomFXDetails _output_get))
            (:Completed
              (bind _asniagara_system (Utilities|Casting|CastToNiagaraSystem _object)
                (:then
                  (Variables|LoadedDetails|SetLoadedParticleSystem _asniagara_system))
                (:CastFailed
                  (Destruction|ClearAppliedVFX (Utilities|Struct|BreakCustomFXDetails (Variables|LoadedDetails|GetLoaded%20Details)))
                  (Variables|LoadedDetails|SetLoadedParticleSystem 0)))
              (if (Utilities|IsValidSoftObjectReference _audio_6_c55694dd4bad4c0b07d0b6ba6026f55b)
                (bind _object_2 (Utilities|AsyncLoadAsset _audio_6_c55694dd4bad4c0b07d0b6ba6026f55b)
                  (:Completed
                    (bind _assound_cue (Utilities|Casting|CastToSoundCue _object_2)
                      (:then
                        (Variables|LoadedDetails|SetLoadedSoundCue _assound_cue))
                      (:CastFailed
                        (Destruction|ClearAppliedAudio (Utilities|Struct|BreakCustomFXDetails (Variables|LoadedDetails|GetLoaded%20Details)))
                        (Variables|LoadedDetails|SetLoadedSoundCue 0)))
                    (CallFunction|CreateandApplyEffectfromLoadedDetails)
                    ;;[0 0.6864 0.7396 0.292] Rerun this function after removing the loaded array entry.
                    (Utilities|Array|RemoveIndex (Variables|Default|GetQueuedFX))
                    (Variables|LoadedDetails|SetLoaded%20Details)
                    (Variables|LoadedDetails|SetLoadedParticleSystem 0)
                    (Variables|LoadedDetails|SetLoadedSoundCue 0)
                    (CallFunction|LoadQueuedEffects)))
                (else
                  _))))
          (else
            _))))))

(event EventEndPlay
  (params (EndPlayReason :type EEndPlayReason%20Enum :default "Destroyed"))
  ;[1 0.1187 0] Destruction: When this component is destroyed also clear all attached effects.
  (Destruction|ClearAllAppliedEffects))

(event Custom|OnPawnDied
  (params (PawnDamageInfo :type Pawn%20Damage%20Data%20Structure%20%28by%20ref%29))
  ;[1 0.3097 0 0.374] On Pawn Death, clear all the applied effects.
  (Destruction|ClearAllAppliedEffects))

(event Custom|OnMinigameChanged
  (params (Minigame :type Fort%20Minigame%20Object%20Reference) (bIsInPlayMode :type Boolean))
  ;[1 0.1744 0 0.499] When the minigame enters or exits play mode, destroy this component.
  (Components|DestroyComponent))

(fn ClearAllAppliedEffects
  ;[1 0.2935 0 0.1] Iterates through all of the effects that are currently playing (both VFX and Audio) and destroys them, then clears the data. Used when the pawn dies or the component is destroyed.
  (for _array_element (Variables|CachedData|GetClientVFX)
    (Utilities|IsValid _array_element
      (:Is%20Valid
        (Components|DestroyComponent _array_element))))
  (Utilities|Array|Clear (Variables|CachedData|GetClientVFX))
  (for _array_element_2 (Variables|CachedData|GetClientFPVFX)
    (Utilities|IsValid _array_element_2
      (:Is%20Valid
        (Components|DestroyComponent _array_element_2))))
  (Utilities|Array|Clear (Variables|CachedData|GetClientFPVFX))
  (for _array_element_3 (Variables|CachedData|GetClientAudio)
    (Utilities|IsValid _array_element_3
      (:Is%20Valid
        (Components|DestroyComponent _array_element_3))))
  (Utilities|Array|Clear (Variables|CachedData|GetClientAudio)))

(fn ClearAppliedAudio
  (params (Index :type Integer))
  (bind _clientaudio (Variables|CachedData|GetClientAudio))
  ;[1 0.4288 0 0.1] If the character has an effect playing at the defined index, this will destroy it.
  (if (Utilities|Array|IsValidIndex _clientaudio Index)
    (bind _output_get (LocalVariables|SetEffect (Utilities|Array|Get(acopy) _clientaudio Index)))
    (Utilities|IsValid _output_get
      (:Is%20Valid
        (Components|DestroyComponent _output_get))
      (:Is%20Not%20Valid
        _))
    (Utilities|Array|SetArrayElem (Variables|CachedData|GetClientAudio) (Variables|GetIndex))))

(fn ClearAppliedVFX
  (params (Index :type Integer))
  (bind _clientvfx (Variables|CachedData|GetClientVFX))
  (bind _clientfpvfx (Variables|CachedData|GetClientFPVFX))
  ;[1 0.4288 0 0.1] If the character has an effect playing at the defined index, this will destroy it.
  (Utilities|FlowControl|Sequence
    (:then_0
      (if (Utilities|Array|IsValidIndex _clientvfx Index)
        (bind _output_get (LocalVariables|SetEffect (Utilities|Array|Get(acopy) _clientvfx Index)))
        (Utilities|IsValid _output_get
          (:Is%20Valid
            (Components|DestroyComponent _output_get))
          (:Is%20Not%20Valid
            _))
        (Utilities|Array|SetArrayElem (Variables|CachedData|GetClientVFX) (Variables|GetIndex))))
    (:then_1
      (if (Utilities|Array|IsValidIndex _clientfpvfx Index)
        (bind _output_get_2 (LocalVariables|SetEffect (Utilities|Array|Get(acopy) _clientfpvfx Index)))
        (Utilities|IsValid _output_get_2
          (:Is%20Valid
            (Components|DestroyComponent _output_get_2))
          (:Is%20Not%20Valid
            _))
        (Utilities|Array|SetArrayElem (Variables|CachedData|GetClientFPVFX) (Variables|GetIndex))))))

(fn ClientSetEffectDetailsFromActor
  (params (Actor :type Actor%20Object%20Reference%20%28by%20ref%29) (Index :type Integer))
  ;[0.0625 1 1 0.433] Populates the supplied index of Custom Effects array (The list of effects that should be active) with the information on the actor from the Device Effects interface.
  ; If the actor doesn't support that interface (because it's a Null entry), then clear the index.
  (bind _ascreative_device_effects_interface (Utilities|Casting|CastToCreative_Device_Effects_Interface Actor)
    (:then
      (bind _effect_color (CreativeEffects|GetCreativeEffectData _ascreative_device_effects_interface :!class /Game/Creative/Devices/Common/Creative_Device_Effects_Interface.Creative_Device_Effects_Interface_C))
      (bind (_custom_system _custom_audio _attachsocket _offsettransform _effectindex) (CreativeEffects|GetCreativeCustomEffectData _ascreative_device_effects_interface :!class /Game/Creative/Devices/Common/Creative_Device_Effects_Interface.Creative_Device_Effects_Interface_C))
      (Variables|SetLocalDetails_VFX_3_0698D8604206BC8D26538AB44AD34AFD (Utilities|Struct|MakeCustomFXDetails _custom_system _custom_audio _attachsocket _offsettransform _effectindex _effect_color)))
    (:CastFailed
      _))
  (Utilities|Array|SetArrayElem (Variables|Default|GetClientCustomEffects) Index (LocalVariables|GetLocalDetails) true)
  (return (Utilities|DetailsHaveValidRef (LocalVariables|GetLocalDetails)) (LocalVariables|GetLocalDetails)))

(fn CreateAndApplyEffectFromLoadedDetails
  (bind _loaded_details (Variables|LoadedDetails|GetLoaded%20Details))
  (bind _index_15_70afae614f69edbdd93259af440374dc (Utilities|Struct|BreakCustomFXDetails _loaded_details))
  (bind _loaded_details_2 (Variables|LoadedDetails|GetLoaded%20Details))
  (bind _index_15_70afae614f69edbdd93259af440374dc_2 (Utilities|Struct|BreakCustomFXDetails _loaded_details_2))
  (bind _loaded_details_3 (Variables|LoadedDetails|GetLoaded%20Details))
  (bind _index_15_70afae614f69edbdd93259af440374dc_3 (Utilities|Struct|BreakCustomFXDetails _loaded_details_3))
  (bind _loaded_details_4 (Variables|LoadedDetails|GetLoaded%20Details))
  (bind _index_15_70afae614f69edbdd93259af440374dc_4 (Utilities|Struct|BreakCustomFXDetails _loaded_details_4))
  (bind _clientfpvfx (Variables|CachedData|GetClientFPVFX))
  (bind _localfindset (LocalVariables|GetLocalFindSet))
  (bind _clientcustomeffects (Variables|Default|GetClientCustomEffects))
  (bind _particlecomp (LocalVariables|GetParticleComp))
  (bind _particlecomp_2 (LocalVariables|GetParticleComp))
  (bind _clientaudio (Variables|CachedData|GetClientAudio))
  (bind _audiocomp (LocalVariables|GetAudioComp))
  (bind _clientvfx (Variables|CachedData|GetClientVFX))
  (bind _particlecomp_3 (LocalVariables|GetParticleComp))
  (bind _particlecomp_4 (LocalVariables|GetParticleComp))
  (Variables|CachedData|GetTargetCharacter
    (:then
      ;[0 0.8678 1 0.215] Check if our loaded details match the expected details (they may have changed during the loading process).
      ; Easiest method for checking struct=struct is to find into a set only containing the info you want to check against.
      (if (Utilities|Array|IsValidIndex _clientcustomeffects _index_15_70afae614f69edbdd93259af440374dc_3)
        (Utilities|Set|Add _localfindset (Utilities|Array|Get(acopy) _clientcustomeffects _index_15_70afae614f69edbdd93259af440374dc_3))
        (if (Utilities|Set|ContainsItem _localfindset (Variables|LoadedDetails|GetLoaded%20Details))
          (Utilities|FlowControl|Sequence
            (:then_0
              ;[1 0.803 0 0.1] Visual Effect.
              ; If we want VFX (Loaded Particle System), check if we already have a particle system component spawned for that index.
              ; If we do, attach it to the appropriate mesh bone and assign the new particle system to it.
              ; If we don't, create one with the particle system and attach to the bone.
              ; Once we have a particle system component, cache it in our array (so future versions can reuse), move it to our desired offset, set the color parameter, and activate.
              (bind _loadedparticlesystem (Variables|LoadedDetails|GetLoadedParticleSystem)
                (:then
                  (if (Utilities|Array|IsValidIndex _clientvfx _index_15_70afae614f69edbdd93259af440374dc_2)
                    (bind _output_get (LocalVariables|SetParticleComp (Utilities|Array|Get(acopy) _clientvfx _index_15_70afae614f69edbdd93259af440374dc_2)))
                    (Utilities|IsValid _output_get
                      (:Is%20Valid
                        (Transformation|DetachFromComponent _particlecomp_3 "KeepRelative" "KeepRelative" "KeepRelative" false)
                        (Transformation|AttachComponentToComponent _particlecomp_3 (Variables|/Script/Engine.Character|GetMesh (Variables|CachedData|GetTargetCharacter)) (Utilities|Struct|BreakCustomFXDetails (Variables|LoadedDetails|GetLoaded%20Details)) "SnapToTarget" "SnapToTarget" "KeepWorld" true)
                        (Niagara|SetNiagaraSystemAsset _particlecomp_3 (Variables|LoadedDetails|GetLoadedParticleSystem) true))
                      (:Is%20Not%20Valid
                        (bind _returnvalue (Niagara|SpawnSystemAttached _loadedparticlesystem (Variables|/Script/Engine.Character|GetMesh (Variables|CachedData|GetTargetCharacter)) (Utilities|Struct|BreakCustomFXDetails (Variables|LoadedDetails|GetLoaded%20Details)) "0, 0, 0" "0, 0, 0" "SnapToTarget" true false "None" false))
                        (LocalVariables|SetParticleComp _returnvalue)
                        (bind _particlecomp_5 (LocalVariables|GetParticleComp)
                          (:then
                            (FirstPerson|SetComponentFirstPersonBehavior (Variables|CachedData|GetTargetCharacter) _particlecomp_5))
                          (:else))))
                    (Utilities|Array|SetArrayElem (Variables|CachedData|GetClientVFX) (Utilities|Struct|BreakCustomFXDetails (Variables|LoadedDetails|GetLoaded%20Details)) _particlecomp_4 true)
                    (Transformation|SetRelativeTransform _particlecomp_4 (Utilities|Struct|BreakCustomFXDetails (Variables|LoadedDetails|GetLoaded%20Details)) false false)
                    (bind _color (Default|GetColor 0 (Variables|LoadedDetails|GetLoaded%20Details)))
                    (Effects|Components|ParticleSystem|SetColorParameter _particlecomp_4 "Color" _color)
                    (Components|Activation|Activate _particlecomp_4 true)))
                (:else)))
            (:then_1
              ;[1 0.803 0 0.1] FirstPerson Visual Effect
              ; Behaves the same as the VFX above, but a separate component that only exists when the player is viewed in first-person mode
              (bind _loadedparticlesystem_2 (Variables|LoadedDetails|GetLoadedParticleSystem)
                (:then
                  (if (Utilities|Array|IsValidIndex _clientfpvfx _index_15_70afae614f69edbdd93259af440374dc)
                    (bind _output_get_2 (LocalVariables|SetParticleComp (Utilities|Array|Get(acopy) _clientfpvfx _index_15_70afae614f69edbdd93259af440374dc)))
                    (Utilities|IsValid _output_get_2
                      (:Is%20Valid
                        (Transformation|DetachFromComponent _particlecomp "KeepRelative" "KeepRelative" "KeepRelative" false)
                        (Transformation|AttachComponentToComponent _particlecomp (Variables|/Script/Engine.Character|GetMesh (Variables|CachedData|GetTargetCharacter)) (Utilities|Struct|BreakCustomFXDetails (Variables|LoadedDetails|GetLoaded%20Details)) "SnapToTarget" "SnapToTarget" "KeepWorld" true)
                        (Niagara|SetNiagaraSystemAsset _particlecomp (Variables|LoadedDetails|GetLoadedParticleSystem) true))
                      (:Is%20Not%20Valid
                        (bind _returnvalue_2 (Niagara|SpawnSystemAttached _loadedparticlesystem_2 (Variables|/Script/Engine.Character|GetMesh (Variables|CachedData|GetTargetCharacter)) (Utilities|Struct|BreakCustomFXDetails (Variables|LoadedDetails|GetLoaded%20Details)) "0, 0, 0" "0, 0, 0" "SnapToTarget" true false "None" false))
                        (LocalVariables|SetParticleComp _returnvalue_2)
                        (bind _particlecomp_6 (LocalVariables|GetParticleComp)
                          (:then
                            (FirstPerson|SetComponentFirstPersonBehavior (Variables|CachedData|GetTargetCharacter) _particlecomp_6 true))
                          (:else))))
                    (Utilities|Array|SetArrayElem (Variables|CachedData|GetClientFPVFX) (Utilities|Struct|BreakCustomFXDetails (Variables|LoadedDetails|GetLoaded%20Details)) _particlecomp_2 true)
                    (Transformation|SetRelativeTransform _particlecomp_2 (Utilities|Struct|BreakCustomFXDetails (Variables|LoadedDetails|GetLoaded%20Details)) false false)
                    (bind _color_2 (Default|GetColor 0 (Variables|LoadedDetails|GetLoaded%20Details)))
                    (Effects|Components|ParticleSystem|SetColorParameter _particlecomp_2 "Color" _color_2)
                    (Components|Activation|Activate _particlecomp_2 true)))
                (:else)))
            (:then_2
              ;[1 0.803 0 0.1] Audio.
              ; If we want Audio (Loaded Sound Cue), check if we already have an audio component spawned for that index.
              ; If we do, attach it to the appropriate mesh bone and assign the new sound to it.
              ; If we don't, create one with the sound cue and attach to the bone.
              ; Once we have an audio component, cache it in our array (so future versions can reuse), move it to our desired offset, and play.
              (bind _loadedsoundcue (Variables|LoadedDetails|GetLoadedSoundCue)
                (:then
                  (if (Utilities|Array|IsValidIndex _clientaudio _index_15_70afae614f69edbdd93259af440374dc_4)
                    (bind _output_get_3 (LocalVariables|SetAudioComp (Utilities|Array|Get(acopy) _clientaudio _index_15_70afae614f69edbdd93259af440374dc_4)))
                    (Transformation|DetachFromComponent _output_get_3 "KeepRelative" "KeepRelative" "KeepRelative" false)
                    (Transformation|AttachComponentToComponent (LocalVariables|GetAudioComp) (Variables|/Script/Engine.Character|GetMesh (Variables|CachedData|GetTargetCharacter)) (Utilities|Struct|BreakCustomFXDetails (Variables|LoadedDetails|GetLoaded%20Details)) "SnapToTarget" "SnapToTarget" "KeepWorld" true)
                    (Audio|Components|Audio|SetSound (LocalVariables|GetAudioComp) (Variables|LoadedDetails|GetLoadedSoundCue))
                    (else
                      (bind _returnvalue_3 (Audio|SpawnSoundAttached :Sound _loadedsoundcue :AttachToComponent (Variables|/Script/Engine.Character|GetMesh (Variables|CachedData|GetTargetCharacter)) :AttachPointName (Utilities|Struct|BreakCustomFXDetails (Variables|LoadedDetails|GetLoaded%20Details)) :LocationType "SnapToTarget" :bStopWhenAttachedToDestroyed true :VolumeMultiplier 1.0 :PitchMultiplier 1.0 :StartTime 0.0 :bAutoDestroy true))
                      (LocalVariables|SetAudioComp _returnvalue_3)))
                  (Utilities|Array|SetArrayElem (Variables|CachedData|GetClientAudio) (Utilities|Struct|BreakCustomFXDetails (Variables|LoadedDetails|GetLoaded%20Details)) _audiocomp true)
                  (Transformation|SetRelativeTransform _audiocomp (Utilities|Struct|BreakCustomFXDetails (Variables|LoadedDetails|GetLoaded%20Details)) false false)
                  (Audio|Components|Audio|Play _audiocomp 0.0))
                (:else)))))))
    (:else)) ; Only load if we have a valid character \n(Player may be out of range and not have a character\nright now. It'll get called when they get into range)
)

(fn GetColor
  (params (Target :type Custom%20FXDetails%20Structure))
  (bind _color_29_2bc5aa9d4f8472f23f453686e9248760 (Utilities|Struct|BreakCustomFXDetails Target))
  (bind (_color_13_d2d2deb24e760961e0d6ab88dc603692 _settype_5_3a2997d544a0618ec312feaed95187b8) (Utilities|Struct|BreakFCreativeEffectDirectColor _color_29_2bc5aa9d4f8472f23f453686e9248760))
  ;[1 0.803 0 0.1] Gets the color based on the details. This can either be the color in the details, or a dynamic color based on either the team (e.g. owning Pawn is on the team 2, which is RED), or the Team Relationship (e.g.the owning Pawn is on the Team 2, and so am I, so show as BLUE for friendly)
  (switch Utilities|FlowControl|Switch|SwitchonECreativeColorSetType _settype_5_3a2997d544a0618ec312feaed95187b8
    (:NewEnumerator1
      (return _color_13_d2d2deb24e760961e0d6ab88dc603692))
    (:NewEnumerator2
      _)
    (:NewEnumerator3
      _))
  (return (Utilities|Array|Get(acopy) (Class|GetClassDefaults "/Game/Creative/Sets/MilitaryBase/Props/CreativeTeamColors.CreativeTeamColors_C") (Teams|GetDynamicTeamColor (Variables|CachedData|GetTargetCharacter) (Variables|CachedData|GetMinigame) _settype_5_3a2997d544a0618ec312feaed95187b8))))

(fn LoadandApplyEffectFromActor
  (params (NewDetails :type Custom%20FXDetails%20Structure%20%28by%20ref%29))
  (bind _newdetails (Variables|GetNewDetails))
  (bind _index_15_70afae614f69edbdd93259af440374dc (Utilities|Struct|BreakCustomFXDetails _newdetails))
  ;[0 1 0.9815 0.324] If the input effect is valid and should show on the pawn for the local player (effects can be flagged as not shown to the local player or not shown to players on other teams) then add the effect to the queue and start loading. Else clear it if it's already applied.
  (if (Utilities|DetailsHaveValidRef NewDetails)
    (bind _show_effect (Default|ShouldShowEffectonPawn 0 NewDetails))
    (if _show_effect
      (Utilities|Array|Add (Variables|Default|GetQueuedFX) (Variables|GetNewDetails))
      (CallFunction|LoadQueuedEffects)
      (else
        _))
    (else
      _))
  (Destruction|ClearAppliedVFX _index_15_70afae614f69edbdd93259af440374dc)
  (Destruction|ClearAppliedAudio _index_15_70afae614f69edbdd93259af440374dc))

(fn LoadEffectfromIndex
  (params (FXIndex :type Integer))
  (bind _fxindex (Variables|GetFXIndex))
  ;[0 0.8905 1 0.209] When a Custom FX Replicates, cache it in our client array and then load or unload any effects.
  (Networking|CanExecuteCosmeticEvents
    (:True
      (bind (_wasvalid _detailsadded) (CallFunction|ClientSetEffectDetailsfromActor (Utilities|Struct|BreakActorBoolPair (Utilities|Select :Index FXIndex :Option%200 (Variables|ReplicatedVariables|GetCustomVFX_0) :Option%201 (Variables|ReplicatedVariables|GetCustomVFX_1) :Option%202 (Variables|ReplicatedVariables|GetCustomVFX_2) :Option%203 (Variables|ReplicatedVariables|GetCustomVFX_3) :Option%204 (Variables|ReplicatedVariables|GetCustomVFX_4))) FXIndex))
      (if _wasvalid
        (CallFunction|LoadAndApplyEffectfromActor _detailsadded)
        (else
          (Destruction|ClearAppliedVFX _fxindex)
          (Destruction|ClearAppliedAudio _fxindex))))))

(fn OnRep_CustomVFX_0
  (CallFunction|LoadEffectFromIndex 0))

(fn OnRep_CustomVFX_1
  (CallFunction|LoadEffectFromIndex 1))

(fn OnRep_CustomVFX_2
  (CallFunction|LoadEffectFromIndex 2))

(fn OnRep_CustomVFX_3
  (CallFunction|LoadEffectFromIndex 3))

(fn OnRep_CustomVFX_4
  (CallFunction|LoadEffectFromIndex 4))

(fn ReinitializeAllEffects
  ;[0 1 0.05902 0.348] Called when the Owning PlayerState gets a new pawn (either they're respawning or have just entered relevancy for the local player).
  ; Update the Target Character then Loop over all of the effects that should be applied and load and apply them.
  (bind _asfort_player_state (Utilities|Casting|CastToFortPlayerState (Components|GetOwner)))
  (bind _ascharacter (Utilities|Casting|CastToCharacter (PlayerState|GetPawn _asfort_player_state)))
  (Variables|CachedData|SetTargetCharacter _ascharacter)
  (for _array_element (Variables|Default|GetClientCustomEffects)
    (CallFunction|LoadAndApplyEffectfromActor _array_element)))

(fn ReplicateCustomFX
  (params (CustomFXActor :type Actor%20Object%20Reference) (Index :type Integer))
  (bind (_customvfx_0_actor_2_9240a989457375e8c719328afc7ca526 _customvfx_0_bool_7_0dceadba4b5ee5a7ee00ac9cdb7fcdee) (Utilities|Struct|BreakActorBoolPair (Variables|GetCustomVFX_0_Actor_2_9240A989457375E8C719328AFC7CA526)))
  (bind (_customvfx_1_actor_2_9240a989457375e8c719328afc7ca526 _customvfx_1_bool_7_0dceadba4b5ee5a7ee00ac9cdb7fcdee) (Utilities|Struct|BreakActorBoolPair (Variables|GetCustomVFX_1_Actor_2_9240A989457375E8C719328AFC7CA526)))
  (bind (_customvfx_2_actor_2_9240a989457375e8c719328afc7ca526 _customvfx_2_bool_7_0dceadba4b5ee5a7ee00ac9cdb7fcdee) (Utilities|Struct|BreakActorBoolPair (Variables|GetCustomVFX_2_Actor_2_9240A989457375E8C719328AFC7CA526)))
  (bind (_customvfx_3_actor_2_9240a989457375e8c719328afc7ca526 _customvfx_3_bool_7_0dceadba4b5ee5a7ee00ac9cdb7fcdee) (Utilities|Struct|BreakActorBoolPair (Variables|GetCustomVFX_3_Actor_2_9240A989457375E8C719328AFC7CA526)))
  (bind (_customvfx_4_actor_2_9240a989457375e8c719328afc7ca526 _customvfx_4_bool_7_0dceadba4b5ee5a7ee00ac9cdb7fcdee) (Utilities|Struct|BreakActorBoolPair (Variables|GetCustomVFX_4_Actor_2_9240A989457375E8C719328AFC7CA526)))
  ;[0 0.7175 1 0.1] Replicating the custom FX. Due to the way RepNotify works this has to be done with a switch rather than a select.
  ; These are individual variables rather than an array for replication optimization (not having to replicate the entire array).
  (Utilities|FlowControl|Sequence
    (:then_0
      (switch int Index
        (:0
          (Variables|SetCustomVFX_0_Actor_2_9240A989457375E8C719328AFC7CA526 (Utilities|Struct|MakeActorBoolPair (Variables|GetCustomFXActor) (not _customvfx_0_bool_7_0dceadba4b5ee5a7ee00ac9cdb7fcdee))))
        (:1
          (Variables|SetCustomVFX_1_Actor_2_9240A989457375E8C719328AFC7CA526 (Utilities|Struct|MakeActorBoolPair (Variables|GetCustomFXActor) (not _customvfx_1_bool_7_0dceadba4b5ee5a7ee00ac9cdb7fcdee))))
        (:2
          (Variables|SetCustomVFX_2_Actor_2_9240A989457375E8C719328AFC7CA526 (Utilities|Struct|MakeActorBoolPair (Variables|GetCustomFXActor) (not _customvfx_2_bool_7_0dceadba4b5ee5a7ee00ac9cdb7fcdee))))
        (:3
          (Variables|SetCustomVFX_3_Actor_2_9240A989457375E8C719328AFC7CA526 (Utilities|Struct|MakeActorBoolPair (Variables|GetCustomFXActor) (not _customvfx_3_bool_7_0dceadba4b5ee5a7ee00ac9cdb7fcdee))))
        (:4
          (Variables|SetCustomVFX_4_Actor_2_9240A989457375E8C719328AFC7CA526 (Utilities|Struct|MakeActorBoolPair (Variables|GetCustomFXActor) (not _customvfx_4_bool_7_0dceadba4b5ee5a7ee00ac9cdb7fcdee))))))
    (:then_1
      (Networking|ForceNetUpdate (Components|GetOwner)))))

(fn ShouldShowEffectOnPawn
  (params (Target :type Custom%20FXDetails%20Structure))
  (bind (_visibletoowningplayer_8_e546f2a4426b8ab99ee118a3125ac217 _teamrelationshipvisibility_11_3bb5265d4d611d2bb5799eb5a6e26aa7) (Utilities|Struct|BreakFCreativeEffectDirectColor (Utilities|Struct|BreakCustomFXDetails Target)))
  (return (and (or (Utilities|NotEqual(Object) (Variables|CachedData|GetTargetCharacter) (Pawn|GetViewTarget (Fortnite|GetFirstLocalFortPlayerController true))) (Utilities|Enum|Equal(Enum) _visibletoowningplayer_8_e546f2a4426b8ab99ee118a3125ac217 "NewEnumerator0")) (Math|Boolean|ORBoolean (and (Team|OnSameTeam (Variables|CachedData|GetTargetCharacter) (Pawn|GetViewTarget (Fortnite|GetFirstLocalFortPlayerController true))) (Utilities|Enum|Equal(Enum) _teamrelationshipvisibility_11_3bb5265d4d611d2bb5799eb5a6e26aa7 "NewEnumerator0")) (and (not (Team|OnSameTeam (Variables|CachedData|GetTargetCharacter) (Pawn|GetViewTarget (Fortnite|GetFirstLocalFortPlayerController true)))) (Utilities|Enum|Equal(Enum) _teamrelationshipvisibility_11_3bb5265d4d611d2bb5799eb5a6e26aa7 "NewEnumerator2")) (Utilities|Enum|Equal(Enum) _teamrelationshipvisibility_11_3bb5265d4d611d2bb5799eb5a6e26aa7 "NewEnumerator1")))))

(fn UpdateActiveCustomEffect
  (params (EffectActor :type Actor%20Object%20Reference) (Apply :type Boolean) (Index :type Integer))
  ;[0 0.3279 1 0.2] Updating the appropriate CustomVFX asset that gets replicated to the client. These are separate variables rather than an array for replication optimization. This could easily be replaced with replicating CustomFXDetails to allow the VFXPowerup to be non-replicated, but leaving it as replicating the actors replicates less information.
  (if Apply
    (CallFunction|ReplicateCustomFX (Variables|GetEffectActor) Index)
    (elif (== EffectActor (Utilities|Struct|BreakActorBoolPair (Utilities|Select :Index Index :Option%200 (Variables|ReplicatedVariables|GetCustomVFX_0) :Option%201 (Variables|ReplicatedVariables|GetCustomVFX_1) :Option%202 (Variables|ReplicatedVariables|GetCustomVFX_2) :Option%203 (Variables|ReplicatedVariables|GetCustomVFX_3) :Option%204 (Variables|ReplicatedVariables|GetCustomVFX_4))))
      (CallFunction|ReplicateCustomFX 0 Index))))

(fn UpdateColorAndVisibilityforAll
  (bind _checkdetails (LocalVariables|GetCheckDetails))
  (bind _checkindex (LocalVariables|GetCheckIndex))
  ;[1 0.803 0 0.1] Called when either the local player or the owning player changes teams. Re-evaluate all of the effects that can be applied for both whether they should be applied and what color they should be.
  ; This is separate from ReinitializeAllEffects since we don't want to restart the particle system if we're just changing the color.
  (for (_array_element _array_index) (Variables|Default|GetClientCustomEffects)
    (bind _output_get (LocalVariables|SetCheckDetails _array_element))
    (LocalVariables|SetCheckIndex _array_index)
    (bind _show_effect (Default|ShouldShowEffectonPawn 0 _output_get))
    (if _show_effect
      (if (Utilities|Array|IsValidIndex (Variables|CachedData|GetClientVFX) (LocalVariables|GetCheckIndex))
        (Utilities|FlowControl|Sequence
          (:then_0
            (bind _output_get_2 (LocalVariables|SetEffectSystem (Utilities|Array|Get(acopy) (Variables|CachedData|GetClientFPVFX) (LocalVariables|GetCheckIndex))))
            (Utilities|IsValid _output_get_2
              (:Is%20Valid
                (Default|GetColor)
                (Effects|Components|ParticleSystem|SetColorParameter :self (LocalVariables|GetEffectSystem) :ParameterName "Color"))) ; It's okay for the element not to exist in\nfirst person, so do nothing in that case.
          )
          (:then_1
            (bind _output_get_3 (LocalVariables|SetEffectSystem (Utilities|Array|Get(acopy) (Variables|CachedData|GetClientVFX) (LocalVariables|GetCheckIndex))))
            (Utilities|IsValid _output_get_3
              (:Is%20Valid
                (bind _color (Default|GetColor 0 _checkdetails))
                (Effects|Components|ParticleSystem|SetColorParameter (LocalVariables|GetEffectSystem) "Color" _color))
              (:Is%20Not%20Valid
                _)) ; Does the element at this index exist?\nIf it does, update its color. If it doesn't, \ncreate it.
          ))
        (else
          _))
      (CallFunction|LoadAndApplyEffectfromActor _checkdetails)
      (else
        (Destruction|ClearAppliedVFX _checkindex)
        (Destruction|ClearAppliedAudio _checkindex)))))
